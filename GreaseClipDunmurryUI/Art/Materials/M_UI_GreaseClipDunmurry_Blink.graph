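# Additive UI material for a blinking circular gauge, apparently a speedometer judging by its SpeedBar and pointer-mask names.
# SDF circles with anti-aliasing form the outer bar, inner bar and masks, while an angle gradient cuts the arc and radial segments form dashes.
# A time sine oscillates the intensity of a powered term to create the pulse.
# A green to yellow-green lerp is linear-dodge blended over multiplied mask terms.
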Material /GreaseClipDunmurryUI/Art/Materials/M_UI_GreaseClipDunmurry_Blink
graph {
 properties {bAutomaticallySetUsageInEditor=false, blendMode=BLEND_Additive, materialDomain=MD_UI, usedShadingModels=Default Lit}
 nodes [
  n1 = NamedRerouteUsage()
  n2 = NamedRerouteUsage()
  MaterialFunctionCall!(
   Blend=LinearInterpolate(A=Constant4Vector() {constant=<rgba 0.0108 1 0 1>}.RGBA, B=Constant4Vector() {constant=<rgba 0.6506 1 0 1>}.RGBA, Alpha=NamedRerouteUsage())
   Base=Multiply(A=Multiply(A=@n1, B=@n2)) {constB=0.5}
  ) {materialFunction=/Engine/Functions/Engine_MaterialFunctions03/Blends/Blend_LinearDodge}
  n3 = Power(Base=NamedRerouteUsage()) {constExponent=3}
  n4 = LinearInterpolate(Alpha=MaterialFunctionCall(Speed=Constant() {r=1}) {materialFunction=&DebugTimeSine}.Result) {constA=2, constB=20}
  Multiply(A=NamedRerouteUsage(), B=Multiply(A=NamedRerouteUsage(), B=Multiply(A=@n3, B=@n4)))
  n5 = MaterialFunctionCall(Size=Constant() {r=0.75}, OutlineThickness=Constant() {r=0.035}) {materialFunction=&MF_UI_Circle}
  NamedRerouteDeclaration(None=MaterialFunctionCall(DistanceField=@n5.SDFOutline) {materialFunction=&MF_UI_AntiAliasing}.Inner) {name=SpeedBar, nodeColor=<rgba 0 0 0 1>}
  NamedRerouteDeclaration(None=OneMinus(None=Add(A=@n5.SDF) {constB=0.75})) {name=Gradient, nodeColor=<rgba 0 0 0 1>}
  n6 = MaterialFunctionCall(Size=Constant() {r=0.68}, OutlineThickness=Constant() {r=0.01}) {materialFunction=&MF_UI_Circle}
  NamedRerouteDeclaration(None=MaterialFunctionCall(DistanceField=@n6.SDFOutline) {materialFunction=&MF_UI_AntiAliasing}.Inner) {name=InnerBar, nodeColor=<rgba 0 0 0 1>}
  n7 = Subtract(A=MaterialFunctionCall!(UVs, Rotation=Constant() {r=0.665}) {materialFunction=/FortUILibrary/MaterialFunctions/Gradients/MF_UI_AngleGradient}.Result) {constB=0.085}
  NamedRerouteDeclaration(None=OneMinus(None=SmoothStep(Value=@n7) {constMax=0.001})) {name=Mask, nodeColor=<rgba 0 0 0 1>}
  NamedRerouteDeclaration(None=MaterialFunctionCall(DistanceField=@n6.SDF, Threshold=Constant() {r=-0.008}) {materialFunction=&MF_UI_AntiAliasing}.Outer) {name=PointerMask, nodeColor=<rgba 0 0 0 1>}
  NamedRerouteDeclaration(None=MaterialFunctionCall(DistanceField=MaterialFunctionCall!(DistanceFieldA=@n5.SDF, DistanceFieldB=@n5.SDF, BlendAmount) {materialFunction=/FortUILibrary/MaterialFunctions/Shapes/Modifiers/MF_UI_SDF_Union}.Exact) {materialFunction=&MF_UI_AntiAliasing}.Inner) {name=DashesMask, nodeColor=<rgba 0 0 0 1>}
  n8 = MaterialFunctionCall!(UV, Segment Count=Constant() {r=12}, Segment Spacing=Constant() {r=0.93}, Fixed Spacing) {materialFunction=/FortUILibrary/MaterialFunctions/Basics/MF_UI_RadialSegments}
  n9 = NamedRerouteUsage()
  n10 = NamedRerouteUsage()
  NamedRerouteDeclaration(None=Multiply(A=SmoothStep(Value=@n8.Segmented Triangles Gradient) {constMax=0.2}, B=Multiply(A=@n9, B=@n10)) {constA=1}) {name=BigDashes, nodeColor=<rgba 0 0 0 1>}
 ]
}
