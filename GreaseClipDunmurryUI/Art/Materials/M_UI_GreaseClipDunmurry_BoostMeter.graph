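# Translucent UI material for a radial boost meter: a ring arc filled according to PercentFill.
# Subtracted radial gradients form the ring, while a rotated radial-arc gradient drives the soft progress mask and Terminus cutoff; radial segments produce a dash mask.
# Fill color blends a two-color linear gradient with an override color over the background.
# The graph also holds a parallel copy of this logic built on a deprecated angle-gradient function.
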Material /GreaseClipDunmurryUI/Art/Materials/M_UI_GreaseClipDunmurry_BoostMeter
graph {
 properties {blendMode=BLEND_Translucent, materialDomain=MD_UI, pixelDepthOffsetMode=PDOM_Legacy, usedShadingModels=Default Lit}
 nodes [
  n1 = VectorParameter() {parameterName=FillColor, defaultValue=<rgba 0 0.604 1 1>}
  n2 = VectorParameter() {parameterName=FillColor2, defaultValue=<rgba 0.1518 1 0 1>}
  n3 = MaterialFunctionCall(
   GradientStart=ScalarParameter() {parameterName=Start}
   GradientEnd=ScalarParameter() {parameterName=End}
   RotationAngle=ScalarParameter() {parameterName=Direction}
  ) {materialFunction=&MF_UI_LinearGradient}
  LinearInterpolate(
   A=VectorParameter() {parameterName=BGColor, defaultValue=<rgba 0 0 0 1>}.RGB
   B=LinearInterpolate(
    A=LinearInterpolate(A=@n1.RGB, B=@n2.RGB, Alpha=@n3.Result)
    B=VectorParameter() {parameterName=FillColorOverride, defaultValue=<rgba 1 1 1 1>}.RGB
    Alpha=ScalarParameter() {parameterName=ColorOverrideScalar}
   )
   Alpha=OneMinus(None=NamedRerouteUsage())
  )
  n4 = ScalarParameter() {defaultValue=0.02, parameterName=Edge}
  n5 = Add(A=@n4, B=Multiply(A=@n4, B=ScalarParameter() {defaultValue=0.2, parameterName=EdgeSoftness}) {constB=1.01})
  n6 = MaterialFunctionCall(
   Alpha=MaterialFunctionCall() {materialFunction=&RadialGradientExponential}.RadialGradientExponential
   Min=@n4
   Max=@n5
  ) {materialFunction=&SmoothStep}
  n7 = Add(A=@n5, B=ScalarParameter() {defaultValue=0.01, parameterName=EdgeWidth})
  n8 = ScalarParameter() {defaultValue=0.2, parameterName=EdgeSoftness}
  n9 = MaterialFunctionCall(
   Alpha=MaterialFunctionCall(CenterPosition=Constant2Vector() {g=0.525, r=0.525}) {materialFunction=&RadialGradientExponential}.RadialGradientExponential
   Min=@n7
   Max=Add(A=@n7, B=Multiply(A=@n7, B=@n8) {constB=1.01})
  ) {materialFunction=&SmoothStep}
  n10 = VectorParameter() {parameterName=BGColor, defaultValue=<rgba 0 0 0 1>}
  n11 = VectorParameter() {parameterName=FillColor, defaultValue=<rgba 0 0.604 1 1>}
  n12 = VectorParameter() {parameterName=FillColorOverride, defaultValue=<rgba 1 1 1 1>}
  n13 = ScalarParameter() {parameterName=ColorOverrideScalar}
  n14 = LinearInterpolate(A=@n11.A, B=@n12.A, Alpha=@n13)
  n15 = MaterialFunctionCall!(UVs, Rotation=ScalarParameter() {defaultValue=-0.62, parameterName=AngleRotation}) {materialFunction=/Game/UI/Foundation/Materials/MF_UI_AngleGradient_DEPRECATED}
  n16 = ScalarParameter() {defaultValue=0.5, parameterName=PercentFill, sliderMax=1}
  n17 = LinearInterpolate(A=OneMinus(None=ScalarParameter() {defaultValue=0.24, parameterName=Terminus}), Alpha=OneMinus(None=Saturate(None=@n16))) {constB=0}
  n18 = Multiply(A=ScalarParameter() {parameterName=PercentEdgeSoftness}, B=@n17)
  n19 = OneMinus(None=MaterialFunctionCall(Alpha=OneMinus(None=@n15.Result), Min=@n17, Max=Add(A=@n17, B=@n18)) {materialFunction=&SmoothStep}.Result) {desc=Fill mask}
  n20 = Constant()
  Multiply(
   A=Multiply(A=Subtract(A=@n6.Result, B=@n9.Result), B=LinearInterpolate(A=@n10.A, B=@n14, Alpha=@n19))
   B=If(A=@n15.Result, B=ScalarParameter() {defaultValue=0.24, parameterName=Terminus}, "A > B"=Constant() {r=1}, "A == B"=@n20, "A < B"=@n20)
  )
  n21 = Constant() {desc=Fill Opacity, r=1}
  n22 = Constant() {desc=Fill Override Color Opacity, r=1}
  n23 = ScalarParameter() {parameterName=ColorOverrideScalar}
  n24 = NamedRerouteUsage()
  Multiply(
   A=Multiply(A=NamedRerouteUsage(), B=NamedRerouteUsage())
   B=LinearInterpolate(A=Constant() {desc=BG Opacity, r=1}, B=LinearInterpolate(A=@n21, B=@n22, Alpha=@n23), Alpha=OneMinus(None=@n24))
  )
  n25 = MaterialFunctionCall!(
   UV=MaterialFunctionCall(UVs=TextureCoordinate(), "Rotation Angle (0-1)"=ScalarParameter() {defaultValue=-0.62, parameterName=AngleRotation}) {materialFunction=&CustomRotator}.Rotated Values
   "Arc Start Angle (0-1)"
   "Arc End Angle (0-1)"
   "Fill Ratio (0-1)"
   Fill Clockwise
  ) {materialFunction=/FortUILibrary/MaterialFunctions/Basics/MF_UI_RadialArc}
  n26 = ScalarParameter() {defaultValue=0.5, parameterName=PercentFill, sliderMax=1}
  n27 = LinearInterpolate(A=OneMinus(None=ScalarParameter() {defaultValue=0.24, parameterName=Terminus}), Alpha=OneMinus(None=Saturate(None=@n26))) {constB=0}
  n28 = Multiply(A=@n27, B=ScalarParameter() {parameterName=PercentEdgeSoftness})
  NamedRerouteDeclaration(None=MaterialFunctionCall(Alpha=OneMinus(None=@n25.Radial Fill), Min=@n27, Max=Add(A=@n27, B=@n28)) {materialFunction=&SmoothStep}.Result) {name=ProgressMask, nodeColor=<rgba 0 0 0 1>}
  n29 = MaterialFunctionCall() {materialFunction=&RadialGradientExponential}
  n30 = ScalarParameter() {defaultValue=0.02, parameterName=Edge}
  n31 = Add(A=@n30, B=Multiply(A=@n30, B=ScalarParameter() {defaultValue=0.2, parameterName=EdgeSoftness}) {constB=1.01})
  n32 = MaterialFunctionCall(CenterPosition=Constant2Vector() {g=0.525, r=0.525}) {materialFunction=&RadialGradientExponential}
  n33 = Add(A=@n31, B=ScalarParameter() {defaultValue=0.01, parameterName=EdgeWidth})
  n34 = Add(A=@n33, B=Multiply(A=@n33, B=ScalarParameter() {defaultValue=0.2, parameterName=EdgeSoftness}) {constB=1.01})
  NamedRerouteDeclaration(None=Subtract(
    A=MaterialFunctionCall(Alpha=@n29.RadialGradientExponential, Min=@n30, Max=@n31) {materialFunction=&SmoothStep}.Result
    B=MaterialFunctionCall(Alpha=@n32.RadialGradientExponential, Min=@n33, Max=@n34) {materialFunction=&SmoothStep}.Result
   )) {name=Shape, nodeColor=<rgba 0 0 0 1>}
  n35 = Constant()
  NamedRerouteDeclaration(None=If(A=@n25.Radial Fill, B=ScalarParameter() {defaultValue=0.24, parameterName=Terminus}, "A > B"=Constant() {r=1}, "A == B"=@n35, "A < B"=@n35)) {name=ShapeMask, nodeColor=<rgba 0 0 0 1>}
  n36 = Constant() {r=24}
  n37 = Constant() {r=0.9}
  NamedRerouteDeclaration(None=SmoothStep(Value=MaterialFunctionCall!(UV, Segment Count=@n36, Segment Spacing=@n37, Fixed Spacing) {materialFunction=/FortUILibrary/MaterialFunctions/Basics/MF_UI_RadialSegments}.Segmented Triangles Gradient) {constMax=0.2}) {name=Dashes, nodeColor=<rgba 0 0 0 1>}
  NamedRerouteUsage()
  n38 = LinearInterpolate(
   A=@n11.RGB
   B=VectorParameter() {parameterName=FillColor2, defaultValue=<rgba 0.1518 1 0 1>}.RGB
   Alpha=MaterialFunctionCall(
    GradientStart=ScalarParameter() {parameterName=Start}
    GradientEnd=ScalarParameter() {parameterName=End}
    RotationAngle=ScalarParameter() {parameterName=Direction}
   ) {materialFunction=&MF_UI_LinearGradient}.Result
  )
  Reroute(None=LinearInterpolate(A=@n10.RGB, B=LinearInterpolate(A=@n38, B=@n12.RGB, Alpha=@n13), Alpha=@n19))
 ]
}
