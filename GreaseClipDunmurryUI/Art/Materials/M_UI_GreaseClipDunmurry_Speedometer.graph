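# Translucent UI material that draws a GreaseClipDunmurry speedometer dial with a needle, speed bar, inner meter and rings.
# Smoothstepped radial gradients form the rings and bands, while angle gradients cut the arc and shape both fills.
# FillPercent drives the speed bar and rotates the needle texture and optional cone glow; InnerMeterFillPercent drives the inner meter separately.
# Named reroutes combine these masks into color and opacity, with the inner meter colored by two colors or a curve atlas.
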Material /GreaseClipDunmurryUI/Art/Materials/M_UI_GreaseClipDunmurry_Speedometer
graph {
 properties {blendMode=BLEND_Translucent, materialDomain=MD_UI, numCustomizedUVs=2, pixelDepthOffsetMode=PDOM_Legacy, usedShadingModels=Default Lit}
 nodes [
  Add(B=LinearInterpolate(B=NamedRerouteUsage(), Alpha=ScalarParameter() {group=Needle, parameterName=EnableGlow}))
  Constant4Vector() {constant=<rgba 0.0762 0.8476 0 0.5>}
  n1 = Multiply(A=NamedRerouteUsage(), B=NamedRerouteUsage())
  LinearInterpolate(A=NamedRerouteUsage(), B=Constant() {desc=Outline and pointer Color, r=1}, Alpha=Max(A=NamedRerouteUsage(), B=@n1))
  MaterialFunctionCall(Texture Scale=ScalarParameter() {defaultValue=0.9225, parameterName=CircleScale}) {materialFunction=&ScaleUVsByCenter}
  n2 = ScalarParameter() {defaultValue=0.38, group=Needle, parameterName=NeedleExtents}
  n3 = LinearInterpolate(A=@n2, B=Multiply(A=@n2) {constB=-1}, Alpha=ScalarParameter() {defaultValue=0.5, group=Needle, parameterName=FillPercent, sliderMax=1})
  MaterialFunctionCall!(UVs, Rotation=@n3) {materialFunction=/Game/Athena/HUD/Vehicles/Materials/MF_UI_AngleGradient2_DEPRECATED}
  n4 = Multiply(A=NamedRerouteUsage(), B=NamedRerouteUsage())
  n5 = NamedRerouteUsage()
  Max(A=Max(A=@n1, B=Multiply(A=@n4, B=@n5) {desc=Mask inner meter}) {constB=0}, B=Multiply(A=NamedRerouteUsage(), B=NamedRerouteUsage()))
  n6 = VectorParameter() {group=Needle, parameterName=GlowScale, defaultValue=<rgba 0.0762 0.8476 0 0.5>}
  n7 = VectorParameter() {group=Needle, parameterName=GlowOffset, defaultValue=<rgba 0.5 1 1 0>}
  n8 = VectorParameter() {group=Needle, parameterName=GlowExtents, defaultValue=<rgba 0.38 -0.38 0 1>}
  n9 = MaterialFunctionCall(
   UVs=Add(A=TextureCoordinate(), B=AppendVector(A=@n6.B, B=@n6.B))
   Rotation Center=AppendVector(A=@n7.R, B=@n7.R)
   "Rotation Angle (0-1)"=LinearInterpolate(A=@n8.R, B=@n8.R, Alpha=ScalarParameter() {defaultValue=0.5, group=Needle, parameterName=FillPercent, sliderMax=1})
  ) {materialFunction=&CustomRotator}
  NamedRerouteDeclaration(None=TextureSample(UVs=MaterialFunctionCall!(Texture Scale=@n6.RGB, UVs=@n9.Rotated Values, Offset XY Texturesize Z=@n7.RGB) {materialFunction=/Engine/Functions/Engine_MaterialFunctions01/Texturing/ScaleUVsAroundPoint}.UVs) {samplerSource=SSM_Clamp_WorldGroupSettings, texture=/Game/Athena/HUD/Vehicles/Materials/T_UI_ConeMask}.RGB) {name=Glow, nodeColor=<rgba 0 0 0 1>}
  NamedRerouteDeclaration(None=MaterialFunctionCall(UVs=TextureCoordinate() {coordinateIndex=1}) {materialFunction=&RadialGradientExponential}.RadialGradientExponential) {name=RadialGradient, nodeColor=<rgba 0 0 0 1>}
  n10 = VectorParameter() {group=Needle, parameterName=FillLineEmptyColor, defaultValue=<rgba 0.0052 0.0052 0.0056 1>}
  NamedRerouteDeclaration(None=@n10.A) {name=FillSpeedBarOpacity, nodeColor=<rgba 0 0 0 1>}
  n11 = ScalarParameter() {defaultValue=0.569, group=InnerMeter, parameterName=InnerMeterWidth}
  n12 = ScalarParameter() {defaultValue=0.02, parameterName=InnerEdgeSoftness}
  n13 = MaterialFunctionCall(
   Alpha=MaterialFunctionCall(UVs=TextureCoordinate() {coordinateIndex=1}) {materialFunction=&RadialGradientExponential}.RadialGradientExponential
   Min=@n11
   Max=Add(A=@n11, B=Multiply(A=@n11, B=@n12) {constB=1.01})
  ) {materialFunction=&SmoothStep}
  n14 = ScalarParameter() {defaultValue=0.38, group=Needle, parameterName=NeedleExtents}
  n15 = LinearInterpolate(
   A=@n14
   B=Multiply(A=@n14) {constB=-1}
   Alpha=ScalarParameter() {defaultValue=1, group=InnerMeter, parameterName=InnerMeterFillPercent, sliderMax=1}
  )
  n16 = Power(
   Base=MaterialFunctionCall!(UVs, Rotation=@n15) {materialFunction=/FortUILibrary/MaterialFunctions/Gradients/MF_UI_AngleGradient}.Result
   Exp=ScalarParameter() {defaultValue=0.66, group=Needle, parameterName=FillGradientPower}
  )
  n17 = Constant()
  n18 = If(
   A=OneMinus(None=Saturate(None=@n16))
   B=Add(A=@n15, B=ScalarParameter() {defaultValue=0.4, group=Needle, parameterName=FillOpacityClamp})
   "A > B"=Constant() {r=1}
   "A == B"=@n17
   "A < B"=@n17
  )
  NamedRerouteDeclaration(None=Multiply(A=NamedRerouteUsage(), B=Multiply(A=@n13.Result, B=@n18))) {name=InnerBarShape, nodeColor=<rgba 0 0 0 1>}
  n19 = MaterialFunctionCall(Rotation=@n3) {materialFunction=&MF_UI_AngleGradient}
  n20 = ScalarParameter() {defaultValue=0.66, group=Needle, parameterName=FillGradientPower}
  n21 = OneMinus(None=Saturate(None=Power(Base=@n19.Result, Exp=@n20)))
  n22 = ScalarParameter() {defaultValue=0.02, parameterName=Edge}
  n23 = Add(A=@n22, B=Multiply(A=@n22, B=ScalarParameter() {defaultValue=0.2, parameterName=EdgeSoftness}) {constB=1.01})
  n24 = MaterialFunctionCall(Alpha=NamedRerouteUsage(), Min=@n22, Max=@n23) {materialFunction=&SmoothStep}
  n25 = Add(B=ScalarParameter() {defaultValue=0.1357, group=Needle, parameterName=FillWidth})
  n26 = ScalarParameter() {defaultValue=0.2, parameterName=EdgeSoftness}
  n27 = MaterialFunctionCall(Alpha=NamedRerouteUsage(), Min=@n25, Max=Add(A=@n25, B=Multiply(A=@n25, B=@n26) {constB=1.01})) {materialFunction=&SmoothStep}
  NamedRerouteDeclaration(None=Multiply(A=@n21, B=Subtract(A=@n24.Result, B=@n27.Result))) {name=SpeedBarGradient, nodeColor=<rgba 0 0 0 1>}
  n28 = ScalarParameter() {defaultValue=0.4, group=Needle, parameterName=FillOpacityClamp}
  n29 = Constant()
  NamedRerouteDeclaration(None=If(A=@n21, B=Add(A=@n3, B=@n28), "A > B"=Constant() {r=1}, "A == B"=@n29, "A < B"=@n29)) {name=SpeedBarProgress, nodeColor=<rgba 0 0 0 1>}
  n30 = ScalarParameter() {defaultValue=0.6286, parameterName=InnerEdge}
  n31 = Add(A=@n30, B=Multiply(A=@n30, B=ScalarParameter() {defaultValue=0.02, parameterName=InnerEdgeSoftness}) {constB=1.01})
  n32 = MaterialFunctionCall(Alpha=NamedRerouteUsage(), Min=@n30, Max=@n31) {materialFunction=&SmoothStep}
  n33 = Add(A=@n31, B=ScalarParameter() {defaultValue=0.02, parameterName=InnerEdgeWidth})
  n34 = ScalarParameter() {defaultValue=0.02, parameterName=InnerEdgeSoftness}
  n35 = MaterialFunctionCall(Alpha=NamedRerouteUsage(), Min=@n33, Max=Add(A=@n33, B=Multiply(A=@n33, B=@n34) {constB=1.01})) {materialFunction=&SmoothStep}
  NamedRerouteDeclaration(None=Multiply(A=Subtract(A=@n32.Result, B=@n35.Result), B=ScalarParameter() {defaultValue=0.6, parameterName=InnerCircleOpacity})) {name=InnerRing, nodeColor=<rgba 0 0 0 1>}
  NamedRerouteDeclaration(None=Multiply(A=OneMinus(None=@n32.Result))) {name=CircleMask, nodeColor=<rgba 0 0 0 1>}
  n36 = AppendVector(A=Constant(), B=Constant() {r=0.5})
  n37 = TextureCoordinate()
  n38 = VectorParameter() {group=Needle, parameterName=NeedleOffset, defaultValue=<rgba 0.5 1 1 0>}
  n39 = ScalarParameter() {defaultValue=0.38, group=Needle, parameterName=NeedleExtents}
  n40 = Multiply(A=@n39) {constB=-1}
  n41 = ScalarParameter() {defaultValue=0.5, group=Needle, parameterName=FillPercent, sliderMax=1}
  n42 = MaterialFunctionCall(
   Texture Scale=AppendVector(A=ScalarParameter() {defaultValue=0.15, parameterName=NeedleScaleX}, B=ScalarParameter() {defaultValue=0.8476, parameterName=NeedleScaleY})
   UVs=MaterialFunctionCall(
    UVs=Add(A=@n36, B=@n37)
    Rotation Center=AppendVector(A=@n38.R, B=@n38.R)
    "Rotation Angle (0-1)"=LinearInterpolate(A=@n39, B=@n40, Alpha=@n41)
   ) {materialFunction=&CustomRotator}.Rotated Values
   Offset XY Texturesize Z=@n38.RGB
  ) {materialFunction=&ScaleUVsAroundPoint}
  n43 = TextureSample(UVs=@n42.UVs) {texture=/GreaseClipDunmurryUI/Art/Textures/T_UI_GreaseClipDunmurry_Needle}
  n44 = ScalarParameter() {defaultValue=0.33, group=Needle, parameterName=NeedleEdge}
  n45 = Multiply(A=@n44, B=ScalarParameter() {defaultValue=0.2, group=Needle, parameterName=NeedleEdgeSoftness}) {constB=1.01}
  NamedRerouteDeclaration(None=MaterialFunctionCall(Alpha=Multiply(A=@n43.R, B=@n42.0-1 mask), Min=@n44, Max=Add(A=@n44, B=@n45)) {materialFunction=&SmoothStep}.Result) {name=Pointer, nodeColor=<rgba 0 0 0 1>}
  n46 = Add(A=@n23, B=ScalarParameter() {defaultValue=0.01, parameterName=EdgeWidth})
  n47 = ScalarParameter() {defaultValue=0.2, parameterName=EdgeSoftness}
  n48 = MaterialFunctionCall(Alpha=NamedRerouteUsage(), Min=@n46, Max=Add(A=@n46, B=Multiply(A=@n46, B=@n47) {constB=1.01})) {materialFunction=&SmoothStep}
  NamedRerouteDeclaration(None=Max(A=Subtract(A=@n24.Result, B=@n48.Result), B=NamedRerouteUsage())) {name=Rings, nodeColor=<rgba 0 0 0 1>}
  n49 = VectorParameter() {group=Needle, parameterName=FillLineColor, defaultValue=<rgba 0 0.1683 0.5029 1>}
  n50 = NamedRerouteUsage()
  n51 = ScalarParameter() {defaultValue=0.38, group=Needle, parameterName=NeedleExtents}
  n52 = LinearInterpolate(
   B=Add(B=ScalarParameter() {defaultValue=0.24, parameterName=Terminus}) {constA=1}
   Alpha=OneMinus(None=MaterialFunctionCall(Rotation=@n51) {materialFunction=&MF_UI_AngleGradient}.Result)
  )
  n53 = LinearInterpolate(
   A=VectorParameter() {group=InnerMeter, parameterName=InnerMeterColor, defaultValue=<rgba 1 1 1 1>}.RGB
   B=VectorParameter() {group=InnerMeter, parameterName=InnerMeterColor2, defaultValue=<rgba 1 1 1 1>}.RGB
   Alpha=@n52
  )
  n54 = CurveAtlasRowParameter(CurveTime=@n52) {
   atlas=/Game/Athena/HUD/Vehicles/Materials/ValetGradientAtlas
   curve=/Game/Athena/HUD/Vehicles/Materials/ValetRPMCurve
   group=InnerMeter
   parameterName=CurveParams
  }
  n55 = ScalarParameter() {defaultValue=1, group=InnerMeter, parameterName=UseGradient, sliderMax=1}
  NamedRerouteDeclaration(None=LinearInterpolate(A=LinearInterpolate(A=@n10.RGB, B=@n49.RGB, Alpha=@n50), B=LinearInterpolate(A=@n53, B=@n54, Alpha=@n55), Alpha=NamedRerouteUsage())) {name=FinalColor, nodeColor=<rgba 0 0 0 1>}
  NamedRerouteDeclaration(None=Multiply(A=NamedRerouteUsage(), B=@n52)) {name=InnerBar, nodeColor=<rgba 0 0 0 1>}
  n56 = NamedRerouteUsage()
  n57 = NamedRerouteUsage()
  n58 = Max(A=Multiply(A=NamedRerouteUsage(), B=NamedRerouteUsage()), B=NamedRerouteUsage())
  n59 = MaterialFunctionCall(Rotation=ScalarParameter() {defaultValue=-0.62, parameterName=AngleRotation}) {materialFunction=&MF_UI_AngleGradient}
  n60 = Power(Base=@n59.Result, Exp=ScalarParameter() {defaultValue=2.6456, parameterName=AnglePower})
  NamedRerouteDeclaration(None=Max(A=Multiply(A=@n56, B=@n57), B=Multiply(A=@n58, B=@n60))) {name=Opacity, nodeColor=<rgba 0 0 0 1>}
  n61 = Constant()
  NamedRerouteDeclaration(None=If(A=@n59.Result, B=ScalarParameter() {defaultValue=0.24, parameterName=Terminus}, "A > B"=Constant() {r=1}, "A == B"=@n61, "A < B"=@n61)) {name=CutMask, nodeColor=<rgba 0 0 0 1>}
 ]
}
