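# UI material for the GreaseClipDunmurry Transform widget, drawing an animated circular gauge with needle, progress arc and tick dashes.
# FillPercent samples a curve atlas to drive arc colors, glow and fill-based effects.
# SDF circles, angle gradients and radial segments build masks shared through named reroutes, while a rotated texture forms the needle.
# Sine pulses and polar-panned noise with explicit derivatives add shimmer, composited with alpha-composite blending.
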Material /GreaseClipDunmurryUI/Art/Materials/M_UI_GreaseClipDunmurry_Transform
graph {
 properties {bAutomaticallySetUsageInEditor=false, blendMode=BLEND_AlphaComposite, materialDomain=MD_UI, usedShadingModels=Default Lit}
 nodes [
  n1 = Multiply(A=NamedRerouteUsage(), B=NamedRerouteUsage())
  n2 = NamedRerouteUsage()
  n3 = Multiply(B=@n2) {constA=0.2, desc=0.20}
  n4 = NamedRerouteUsage()
  n5 = SmoothStep(Value=NamedRerouteUsage()) {constMax=0.3, constMin=0.29}
  n6 = Saturate(None=Subtract(A=Multiply(A=@n2, B=@n4), B=@n5) {constB=0})
  n7 = NamedRerouteUsage()
  n8 = NamedRerouteUsage()
  n9 = Multiply(A=@n8) {constB=0.3}
  n10 = Max(
   A=Max(A=LinearInterpolate(A=@n3, B=@n2, Alpha=@n6) {constA=0.3, constB=0.1}, B=NamedRerouteUsage())
   B=Max(A=Multiply(A=@n7) {constB=0.6}, B=LinearInterpolate(A=@n9, B=@n8, Alpha=@n6) {constA=0.2})
  )
  n11 = NamedRerouteUsage()
  n12 = NamedRerouteUsage()
  n13 = Constant() {r=0.15}
  n14 = Constant() {r=0.9}
  n15 = Constant() {r=0.25}
  n16 = Multiply(
   A=OneMinus(None=@n12)
   B=OneMinus(None=MaterialFunctionCall(GradientStart=@n13, GradientEnd=@n14, RotationAngle=@n15) {materialFunction=&MF_UI_LinearGradient}.Result)
  )
  n17 = Multiply(A=NamedRerouteUsage()) {constB=3}
  n18 = NamedRerouteUsage()
  n19 = Subtract(A=Saturate(None=Multiply(A=@n17, B=@n18)), B=NamedRerouteUsage())
  n20 = Saturate(None=Subtract(A=Multiply(A=@n16) {constB=0.5}, B=Saturate(None=@n19)))
  n21 = Max(A=@n12, B=NamedRerouteUsage())
  n22 = NamedRerouteUsage()
  n23 = Constant()
  n24 = Constant() {r=0.6}
  n25 = Constant() {r=0.25}
  n26 = Multiply(
   A=OneMinus(None=Max(A=@n21, B=@n22))
   B=OneMinus(None=MaterialFunctionCall(GradientStart=@n23, GradientEnd=@n24, RotationAngle=@n25) {materialFunction=&MF_UI_LinearGradient}.Result)
  )
  n27 = Multiply(A=@n26) {constB=0.5}
  Max(A=Max(A=@n1, B=Multiply(A=@n10, B=@n11)), B=Saturate(None=Add(A=@n20, B=@n27)))
  n28 = LinearInterpolate(
   A=Constant3Vector() {constant=<rgba 0.1721 0.8975 1 1>}
   B=Constant3Vector() {constant=<rgba 1 0.7091 0 1>}
   Alpha=MaterialFunctionCall(Speed=Constant() {r=1}) {materialFunction=&DebugTimeSine}.Result
  )
  n29 = NamedRerouteUsage()
  n30 = CurveAtlasRowParameter(CurveTime=ScalarParameter() {group=Needle, parameterName=FillPercent, sliderMax=1}) {
   atlas=&/CurveAtlas_UI_GreaseClipDunmurry
   curve=&/Curve_UI_GreaseClipDunmurry
   parameterName=Curve
  }
  n31 = NamedRerouteUsage()
  n32 = Multiply(A=NamedRerouteUsage()) {constB=0.4}
  n33 = Multiply(
   A=LinearInterpolate(
    A=@n28
    B=MaterialFunctionCall!(Blend=@n29, Base=@n28) {materialFunction=/Engine/Functions/Engine_MaterialFunctions03/Blends/Blend_Screen}.Result
    Alpha=@n30.R
   )
   B=Saturate(None=Subtract(A=@n31, B=@n32))
  )
  n34 = NamedRerouteUsage()
  n35 = Constant() {r=0.4}
  n36 = NamedRerouteUsage()
  n37 = LinearInterpolate(B=NamedRerouteUsage(), Alpha=Saturate(None=Subtract(A=@n36, B=@n5) {constB=0}))
  n38 = NamedRerouteUsage()
  n39 = Multiply(A=NamedRerouteUsage(), B=OneMinus(None=NamedRerouteUsage()))
  n40 = Max(A=Add(A=NamedRerouteUsage(), B=Multiply(A=@n39) {constB=0.2}), B=NamedRerouteUsage())
  n41 = NamedRerouteUsage()
  n42 = LinearInterpolate(
   A=MaterialFunctionCall!(Blend=Multiply(A=@n34, B=@n35), Base=Multiply(A=@n37, B=@n38)) {materialFunction=/Engine/Functions/Engine_MaterialFunctions03/Blends/Blend_ColorDodge}.Result
   Alpha=Max(A=Multiply(A=@n40, B=@n41), B=@n1)
  )
  Multiply(A=Add(A=Add(A=@n33, B=@n42), B=NamedRerouteUsage()), B=VertexColor().A)
  n43 = AppendVector(A=Constant(), B=Constant() {r=0.5})
  n44 = TextureCoordinate()
  n45 = VectorParameter() {group=Needle, parameterName=NeedleOffset, defaultValue=<rgba 0.5 1 1 0>}
  n46 = Constant() {desc=RotationStartPoint, r=0.337}
  n47 = Constant() {desc=RotationEndPoint, r=-0.332}
  n48 = NamedRerouteUsage()
  n49 = MaterialFunctionCall!(
   Texture Scale=AppendVector(A=Constant() {r=0.25}, B=Constant() {r=0.75})
   UVs=MaterialFunctionCall(
    UVs=Add(A=@n43, B=@n44)
    Rotation Center=AppendVector(A=@n45.R, B=@n45.R)
    "Rotation Angle (0-1)"=LinearInterpolate(A=@n46, B=@n47, Alpha=@n48) {constA=0.35, constB=-0.35}
   ) {materialFunction=&CustomRotator}.Rotated Values
   Offset XY Texturesize Z=@n45.RGB
  ) {materialFunction=/Engine/Functions/Engine_MaterialFunctions01/Texturing/ScaleUVsAroundPoint}
  n50 = Multiply(A=TextureSample(UVs=@n49.UVs) {texture=/GreaseClipDunmurryUI/Art/Textures/T_UI_GreaseClipDunmurry_Needle}.R, B=@n49.0-1 mask)
  n51 = ScalarParameter() {defaultValue=0.33, group=Needle, parameterName=NeedleEdge}
  n52 = Add(A=@n51, B=Multiply(A=@n51, B=ScalarParameter() {defaultValue=0.2, group=Needle, parameterName=NeedleEdgeSoftness}) {constB=1.01})
  NamedRerouteDeclaration(None=Saturate(None=MaterialFunctionCall(Alpha=@n50, Min=@n51, Max=@n52) {materialFunction=&SmoothStep}.Result)) {name=Pointer, nodeColor=<rgba 0 0 0 1>}
  n53 = MaterialFunctionCall(Size=Constant() {r=0.75}, OutlineThickness=Constant() {r=0.035}) {materialFunction=&MF_UI_Circle}
  NamedRerouteDeclaration(None=MaterialFunctionCall(DistanceField=@n53.SDFOutline) {materialFunction=&MF_UI_AntiAliasing}.Inner) {name=SpeedBar, nodeColor=<rgba 0 0 0 1>}
  n54 = Constant3Vector() {constant=<rgba 0.9473 0.107 0.008 1>}
  n55 = Constant3Vector() {constant=<rgba 0.2961 0.0262 0.0044 1>}
  n56 = NamedRerouteUsage()
  n57 = Constant3Vector() {constant=<rgba 1 0.9734 0.0802 1>}
  n58 = Constant3Vector() {constant=<rgba 0.6514 0.3342 0.107 1>}
  n59 = NamedRerouteUsage()
  n60 = Constant3Vector() {constant=<rgba 0.5711 0.9911 0.0452 1>}
  n61 = Constant3Vector() {constant=<rgba 0.0595 0.2159 0.0222 1>}
  n62 = Constant() {r=0.335}
  n63 = CurveAtlasRowParameter(CurveTime=ScalarParameter() {group=Needle, parameterName=FillPercent, sliderMax=1}) {
   atlas=&CurveAtlas_UI_GreaseClipDunmurry
   curve=&Curve_UI_GreaseClipDunmurry
   parameterName=Curve
  }
  n64 = LinearInterpolate(Alpha=@n63.G) {constA=0.99, constB=0.33, desc=GradientStart}
  n65 = Divide(
   A=Subtract(A=MaterialFunctionCall!(UVs, Rotation=@n62) {materialFunction=/FortUILibrary/MaterialFunctions/Gradients/MF_UI_AngleGradient}.Result, B=@n64)
   B=Subtract(A=Constant() {desc=Gradient End, r=1}, B=@n64)
  )
  n66 = Multiply(A=Power(Base=@n63.G) {constExponent=2.5}) {constB=2}
  NamedRerouteDeclaration(None=MaterialFunctionCall(
    A=LinearInterpolate(A=@n54, B=@n55, Alpha=@n56)
    B=LinearInterpolate(A=@n57, B=@n58, Alpha=@n59)
    C=LinearInterpolate(A=@n60, B=@n61, Alpha=@n65)
    Alpha=Saturate(None=@n66)
   ) {materialFunction=&3ColorBlend}.Result) {name=ProgressBarColors, nodeColor=<rgba 0 0 0 1>}
  NamedRerouteDeclaration(None=@n65) {name=ColorGradient, nodeColor=<rgba 0 0 0 1>}
  NamedRerouteDeclaration(None=@n50) {name=PointerBorder, nodeColor=<rgba 0.1484 0 1 1>}
  n67 = CurveAtlasRowParameter(CurveTime=ScalarParameter() {group=Needle, parameterName=FillPercent, sliderMax=1}) {
   atlas=&CurveAtlas_UI_GreaseClipDunmurry
   curve=&Curve_UI_GreaseClipDunmurry
   parameterName=Curve
  }
  NamedRerouteDeclaration(None=@n67.G) {name=FillPercent, nodeColor=<rgba 0 0 0 1>}
  NamedRerouteDeclaration(None=OneMinus(None=Add(A=@n53.SDF) {constB=0.75})) {name=Gradient, nodeColor=<rgba 0 0 0 1>}
  n68 = Subtract(A=MaterialFunctionCall(Rotation=Constant() {r=0.665}) {materialFunction=&MF_UI_AngleGradient}.Result) {constB=0.085}
  NamedRerouteDeclaration(None=OneMinus(None=SmoothStep(Value=@n68) {constMax=0.001})) {name=GaugeAdditiveMask, nodeColor=<rgba 0 0 0 1>}
  n69 = LinearInterpolate(A=Constant4Vector() {constant=<rgba 0.0108 1 0 1>}.RGBA, B=Constant4Vector() {constant=<rgba 0.6506 1 0 1>}.RGBA, Alpha=NamedRerouteUsage())
  n70 = NamedRerouteUsage()
  n71 = Multiply(A=Multiply(A=NamedRerouteUsage(), B=@n70)) {constB=0.5}
  n72 = NamedRerouteUsage()
  n73 = NamedRerouteUsage()
  n74 = MaterialFunctionCall(Speed=Constant() {r=1}) {materialFunction=&DebugTimeSine}
  n75 = Multiply(A=@n70, B=Multiply(A=Power(Base=@n73) {constExponent=3}, B=LinearInterpolate(Alpha=@n74.Result) {constA=2, constB=20}))
  NamedRerouteDeclaration(None=Multiply(
    A=MaterialFunctionCall!(Blend=@n69, Base=@n71) {materialFunction=/Engine/Functions/Engine_MaterialFunctions03/Blends/Blend_LinearDodge}.Result
    B=Multiply(A=@n72, B=@n75)
   )) {name=Additive, nodeColor=<rgba 0.9219 1 0 1>}
  n76 = Constant() {r=0.4}
  n77 = Constant() {r=0.8}
  n78 = CurveAtlasRowParameter(CurveTime=ScalarParameter() {group=Needle, parameterName=FillPercent, sliderMax=1}) {
   atlas=&CurveAtlas_UI_GreaseClipDunmurry
   curve=&Curve_UI_GreaseClipDunmurry
   parameterName=Curve
  }
  n79 = LinearInterpolate(
   B=LinearInterpolate(Alpha=MaterialFunctionCall(Speed=@n77) {materialFunction=&DebugTimeSine}.Result) {constA=3, constB=5}
   Alpha=@n78.G
  ) {constA=10, constAlpha=1, constB=3}
  n80 = Multiply(
   A=OneMinus(None=NamedRerouteUsage())
   B=OneMinus(None=MaterialFunctionCall(Radius=@n76, Density=@n79) {materialFunction=&RadialGradientExponential}.RadialGradientExponential)
  )
  NamedRerouteDeclaration(None=Multiply(A=Multiply(A=@n80, B=@n78.G)) {constB=0.5}) {name=InnerGlow, nodeColor=<rgba 0 0 0 1>}
  NamedRerouteDeclaration(None=@n26) {name=NumbersBase, nodeColor=<rgba 0 0 0 1>}
  n81 = MaterialFunctionCall() {materialFunction=&VectorToRadialValue}
  n82 = TextureCoordinate()
  n83 = DDX(Value=@n82)
  n84 = DDY(Value=@n82)
  n85 = TextureSample!(
   UVs=Multiply(A=Panner(Coordinate=@n81.Vector Converted to Angle) {speedX=0.15}) {constB=2}
   Tex
   "DDX(UVs)"=@n83
   "DDY(UVs)"=@n84
   Apply View MipBias
  ) {mipValueMode=TMVM_Derivative, samplerType=SAMPLERTYPE_Grayscale, texture=&LowResBlurredNoise}
  n86 = TextureSample(
   UVs=Multiply(A=Panner(Coordinate=@n81.Vector Converted to Angle) {speedX=-0.125, speedY=-0.125}) {constB=2}
   "DDX(UVs)"=@n83
   "DDY(UVs)"=@n84
  ) {mipValueMode=TMVM_Derivative, samplerType=SAMPLERTYPE_Grayscale, texture=&LowResBlurredNoise}
  n87 = Add(A=@n85.R, B=@n86.R)
  NamedRerouteDeclaration(None=SmoothStep(Value=@n87) {constMin=0.5}) {name=Shine, nodeColor=<rgba 0 0 0 1>}
  n88 = MaterialFunctionCall(Size=Constant() {r=0.68}, OutlineThickness=Constant() {r=0.01}) {materialFunction=&MF_UI_Circle}
  NamedRerouteDeclaration(None=MaterialFunctionCall(DistanceField=@n88.SDFOutline) {materialFunction=&MF_UI_AntiAliasing}.Inner) {name=InnerBar, nodeColor=<rgba 0 0 0 1>}
  n89 = MaterialFunctionCall(Rotation=Constant() {r=0.335}) {materialFunction=&MF_UI_AngleGradient}
  NamedRerouteDeclaration(None=SmoothStep(Value=Subtract(A=@n89.Result) {constB=0.33}) {constMax=0.001}) {name=Mask, nodeColor=<rgba 0 0 0 1>}
  NamedRerouteDeclaration(None=MaterialFunctionCall(DistanceField=@n88.SDF, Threshold=Constant() {r=-0.008}) {materialFunction=&MF_UI_AntiAliasing}.Outer) {name=PointerMask, nodeColor=<rgba 0 0 0 1>}
  n90 = NamedRerouteUsage()
  NamedRerouteDeclaration(None=Step(Y=OneMinus(None=@n89.Result), X=LinearInterpolate(Alpha=@n90) {constB=0.67})) {name=Progress, nodeColor=<rgba 0 0 0 1>}
  n91 = TextureCoordinate()
  n92 = Constant2Vector() {g=0.06, r=0.5}
  n93 = MaterialFunctionCall(UVs=Multiply(A=Add(A=@n91, B=@n92), B=Constant2Vector() {g=1, r=0.5}), Size=Constant() {r=0.75}) {materialFunction=&MF_UI_Circle}
  NamedRerouteDeclaration(None=MaterialFunctionCall(DistanceField=Add(A=@n93.SDF) {constB=0.4}) {materialFunction=&MF_UI_AntiAliasing}.Inner) {name=InnerCircleMask, nodeColor=<rgba 0 0 0 1>}
  n94 = MaterialFunctionCall!(UV, Segment Count=Constant() {r=48}, Segment Spacing, Fixed Spacing) {materialFunction=/FortUILibrary/MaterialFunctions/Basics/MF_UI_RadialSegments}
  NamedRerouteDeclaration(None=Multiply(A=NamedRerouteUsage(), B=SmoothStep(Value=@n94.Segmented Triangles Gradient) {constMax=0.2})) {name=SmallDashes, nodeColor=<rgba 0 0 0 1>}
  NamedRerouteDeclaration(None=MaterialFunctionCall(DistanceField=MaterialFunctionCall!(DistanceFieldA=@n53.SDF, DistanceFieldB=@n53.SDF, BlendAmount) {materialFunction=/FortUILibrary/MaterialFunctions/Shapes/Modifiers/MF_UI_SDF_Union}.Exact) {materialFunction=&MF_UI_AntiAliasing}.Inner) {name=DashesMask, nodeColor=<rgba 0 0 0 1>}
  n95 = MaterialFunctionCall(Segment Count=Constant() {r=12}, Segment Spacing=Constant() {r=0.93}) {materialFunction=&MF_UI_RadialSegments}
  n96 = NamedRerouteUsage()
  n97 = NamedRerouteUsage()
  NamedRerouteDeclaration(None=Multiply(A=SmoothStep(Value=@n95.Segmented Triangles Gradient) {constMax=0.2}, B=Multiply(A=@n96, B=@n97)) {constA=1}) {name=BigDashes, nodeColor=<rgba 0 0 0 1>}
  Saturate(None=@n87)
 ]
}
