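# Translucent UI material for a GreaseClipDunmurry VFX element that draws a pale cyan circle outline.
# MF_UI_Circle supplies an outline SDF from Scale and OutlineThickness parameters, which a narrow SmoothStep and OneMinus turn into a crisp ring mask.
# That mask is scaled by an Opacity parameter and multiplied by MF_UI_IsolateUV on a two-by-one grid, which appears to limit the ring to one half.
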Material /GreaseClipDunmurryUI/Art/Materials/M_UI_GreaseClipDunmurry_Vfx
graph {
 properties {
  bAutomaticallySetUsageInEditor=false
  blendMode=BLEND_Translucent
  materialDomain=MD_UI
  pixelDepthOffsetMode=PDOM_Legacy
  usedShadingModels=Default Lit
 }
 nodes [
  Constant3Vector() {constant=<rgba 0.6871 0.8854 0.8575 1>}
  n1 = SmoothStep(Value=MaterialFunctionCall(
    Size=ScalarParameter() {defaultValue=0.8, parameterName=Scale}
    OutlineThickness=ScalarParameter() {defaultValue=0.104, parameterName=OutlineThickness}
   ) {materialFunction=&MF_UI_Circle}.SDFOutline) {constMax=0.01}
  Multiply(
   A=Multiply(A=OneMinus(None=@n1), B=ScalarParameter() {defaultValue=1, parameterName=Opacity})
   B=MaterialFunctionCall!(UV, Grid Size=Constant2Vector() {g=1, r=2}, Index) {materialFunction=/FortUILibrary/MaterialFunctions/Basics/MF_UI_IsolateUV}.Result
  )
 ]
}
