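# UI material that draws a flux capacitor graphic with a dim idle look and an animated glowing active look.
# An SDF texture's R, G and B channels become antialiased circle, ring and line masks plus soft glows, sampled on center-scaled UVs.
# A time-scrolled cosine ripple combines with an SDF for flux glow, alongside sine pulses and color-dodge tinting.
# GlowActivated raises opacity and drives a curve-atlas blend between both looks.
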Material /GreaseClipDunmurryUI/Art/Materials/M_UI_GreaseClipDunmury_FluxCapacitor
graph {
 properties {bAutomaticallySetUsageInEditor=false, blendMode=BLEND_AlphaComposite, materialDomain=MD_UI, usedShadingModels=Default Lit}
 nodes [
  Constant() {r=1}
  MaterialFunctionCall() {materialFunction=&GetUserInterfaceUV}
  MaterialFunctionCall() {materialFunction=&GetUserInterfaceUV}
  MaterialFunctionCall!(Anchor, Translation, Padding, Scale=Constant() {r=1.5}) {materialFunction=/FortUILibrary/MaterialFunctions/Basics/MF_UI_VertexTransform}
  n1 = Constant3Vector() {constant=<rgba 0.8879 0.8308 0.5271 1>}
  n2 = NamedRerouteUsage()
  n3 = NamedRerouteUsage()
  n4 = Multiply(
   A=Max(A=Max(A=@n2, B=@n3), B=NamedRerouteUsage())
   B=LinearInterpolate(A=Constant() {desc=Opacity, r=0.1}, Alpha=ScalarParameter() {parameterName=GlowActivated}) {constA=0.1}
  )
  n5 = NamedRerouteUsage()
  n6 = Constant3Vector() {constant=<rgba 0.5417 0.4894 0.2116 1>}
  n7 = Constant3Vector() {constant=<rgba 0 0.6172 1 1>}
  n8 = NamedRerouteUsage()
  n9 = MaterialFunctionCall(Speed=Constant() {r=1.4}) {materialFunction=&DebugTimeSine}
  n10 = LinearInterpolate(
   A=MaterialFunctionCall!(Blend=@n5, Base=LinearInterpolate(A=@n6, B=@n7, Alpha=@n5)) {materialFunction=/Engine/Functions/Engine_MaterialFunctions03/Blends/Blend_ColorDodge}.Result
   B=@n1
   Alpha=Max(A=Multiply(A=@n8, B=@n9.Result), B=NamedRerouteUsage())
  ) {constA=1}
  n11 = NamedRerouteUsage()
  n12 = Multiply(A=NamedRerouteUsage()) {constB=0.8}
  n13 = Constant() {r=1.4}
  n14 = Add(
   A=Add(A=@n4, B=NamedRerouteUsage())
   B=Multiply(
    A=Max(A=@n11, B=@n12)
    B=MaterialFunctionCall(Speed=@n13) {materialFunction=&DebugTimeSine}.Result
   )
  )
  n15 = CurveAtlasRowParameter(CurveTime=ScalarParameter() {parameterName=GlowActivated}) {
   atlas=&/CurveAtlas_UI_GreaseClipDunmurry
   curve=&/Curve_UI_GreaseClipDunmurry
   parameterName=Curve
  }
  Multiply(A=LinearInterpolate(A=Multiply(A=@n1, B=@n4), B=Multiply(A=@n10, B=@n14), Alpha=@n15.R), B=VertexColor().A)
  n16 = TextureSample(UVs=NamedRerouteUsage()) {samplerSource=SSM_Clamp_WorldGroupSettings, texture=/GreaseClipDunmurryUI/Art/Textures/T_UI_GreaseClipDunmury_Flux_SDF}
  n17 = Constant() {r=0.05}
  Multiply(A=MaterialFunctionCall(
    DistanceField=MaterialFunctionCall!(DistanceField=@n16.G, Threshold, Thickness=@n17, Alignment, Smoothing) {materialFunction=/FortUILibrary/MaterialFunctions/Shapes/Modifiers/MF_UI_ShapeOutline}.SDFOutline
    Smoothing=Constant() {r=0.1}
   ) {materialFunction=&MF_UI_AntiAliasing}.Outer) {constB=0.5}
  n18 = MaterialFunctionCall(UVs=Add(A=NamedRerouteUsage(), B=Constant2Vector() {g=0.06, r=-0.01})) {materialFunction=&MF_UI_Circle}
  n19 = Frac(None=Multiply(A=Time(), B=Constant() {desc=speed, r=-0.7}) {constB=0.5})
  n20 = Cosine(None=Multiply(A=Subtract(A=@n18.SDF, B=@n19), B=Constant() {desc=Repetitions, r=2}))
  Multiply(A=@n20) {constB=0.5, desc="cosine range is -1 to 1, so we need to bring it back to 0 to 1"}
  Multiply() {constB=0.2}
  n21 = Constant() {r=0.4}
  n22 = Constant() {r=0.03}
  NamedRerouteDeclaration(None=MaterialFunctionCall(DistanceField=@n16.R, Threshold=@n21, Smoothing=@n22) {materialFunction=&MF_UI_AntiAliasing}.Outer) {name=Circles, nodeColor=<rgba 0 0 0 1>}
  NamedRerouteDeclaration(None=MaterialFunctionCall(DistanceField=@n16.G, Threshold=@n21, Smoothing=@n22) {materialFunction=&MF_UI_AntiAliasing}.Outer) {name=Rings, nodeColor=<rgba 0 0 0 1>}
  n23 = Constant() {r=0.5}
  n24 = Constant() {r=0.8}
  NamedRerouteDeclaration(None=MaterialFunctionCall(DistanceField=@n16.B, Threshold=@n23, Smoothing=@n24) {materialFunction=&MF_UI_AntiAliasing}.Outer) {name=Lines_Glow_1, nodeColor=<rgba 0 0 0 1>}
  NamedRerouteDeclaration(None=MaterialFunctionCall(
    UVs=MaterialFunctionCall() {materialFunction=&GetUserInterfaceUV}.Normalized UV
    Texture Scale=ScalarParameter() {defaultValue=0.75, parameterName=Scale}
   ) {materialFunction=&ScaleUVsByCenter}.UVs) {name=UV, nodeColor=<rgba 0 0 0 1>}
  n25 = Add(
   A=Multiply(A=@n20) {constB=0.25, desc="cosine range is -1 to 1, so we need to bring it back to 0 to 1"}
   B=Constant() {desc=Thickness, r=0.3}
  ) {constB=0.2}
  n26 = Power(Base=NamedRerouteUsage()) {constExponent=1}
  NamedRerouteDeclaration(None=MaterialFunctionCall!(DistanceFieldA=@n25, DistanceFieldB=Multiply(A=@n26) {constB=3}, BlendAmount) {materialFunction=/FortUILibrary/MaterialFunctions/Shapes/Modifiers/MF_UI_SDF_Exclusion}.Exact) {name=FluxGlow_1, nodeColor=<rgba 0 0 0 1>}
  NamedRerouteDeclaration(None=MaterialFunctionCall(DistanceField=@n16.B, Threshold=@n21, Smoothing=@n22) {materialFunction=&MF_UI_AntiAliasing}.Outer) {name=Lines, nodeColor=<rgba 0 0 0 1>}
  n27 = Constant() {r=0.3}
  n28 = Constant() {r=0.6}
  NamedRerouteDeclaration(None=MaterialFunctionCall(DistanceField=@n16.R, Threshold=@n27, Smoothing=@n28) {materialFunction=&MF_UI_AntiAliasing}.Outer) {name=Circles_Glow, nodeColor=<rgba 0 0 0 1>}
  NamedRerouteDeclaration(None=MaterialFunctionCall(DistanceField=@n16.G, Threshold=@n27, Smoothing=@n28) {materialFunction=&MF_UI_AntiAliasing}.Outer) {name=Rings_Glow, nodeColor=<rgba 0 0 0 1>}
  NamedRerouteDeclaration(None=MaterialFunctionCall(DistanceField=@n16.B, Threshold=@n27, Smoothing=@n28) {materialFunction=&MF_UI_AntiAliasing}.Outer) {name=Lines_Glow, nodeColor=<rgba 0 0 0 1>}
  NamedRerouteDeclaration(None=@n16.B) {name=LinesSDF, nodeColor=<rgba 0 0 0 1>}
  NamedRerouteDeclaration(None=MaterialFunctionCall(DistanceField=@n16.R, Threshold=@n23, Smoothing=@n24) {materialFunction=&MF_UI_AntiAliasing}.Outer) {name=Circles_Glow_1, nodeColor=<rgba 0 0 0 1>}
  NamedRerouteDeclaration(None=MaterialFunctionCall(DistanceField=@n16.G, Threshold=@n23, Smoothing=Constant() {r=1}) {materialFunction=&MF_UI_AntiAliasing}.Outer) {name=Rings_Glow_1, nodeColor=<rgba 0 0 0 1>}
  SmoothStep(Value=@n25) {constMax=0.6, constMin=0.55}
  SmoothStep(Value=@n25) {constMax=0.8, constMin=0.4}
 ]
}
