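# Additive UI material that draws a dynamic numeric readout, labeled as a speed number, from the Number parameter clamped to 99.
# UV math splits the width into digit cells, using log10 and powers of ten to pick each digit's row in a ten-row distance-field atlas.
# A mirrored, fixed-row sample forms a background number, and both samples are anti-aliased.
# Dim and bright three-color blends shift red, amber, green as a value rises.
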Material /GreaseClipDunmurryUI/Art/Materials/M_UI_GreaseClipDunmury_Numbers
graph {
 properties {bAutomaticallySetUsageInEditor=false, blendMode=BLEND_Additive, materialDomain=MD_UI, usedShadingModels=Default Lit}
 nodes [
  n1 = NamedRerouteUsage()
  LinearInterpolate(A=@n1, B=NamedRerouteUsage())
  n2 = Saturate(None=Divide(A=NamedRerouteUsage()) {constB=80})
  LinearInterpolate(
   A=MaterialFunctionCall(
    A=Constant3Vector() {constant=<rgba 0.063 0.0027 0.0024 1>}
    B=Constant3Vector() {constant=<rgba 0.0762 0.0284 0.0048 1>}
    C=Constant3Vector() {constant=<rgba 0.007 0.0423 0.0116 1>}
    Alpha=@n2
   ) {materialFunction=&3ColorBlend}.Result
   B=MaterialFunctionCall(
    A=Constant3Vector() {constant=<rgba 0.9473 0.107 0.008 1>}
    B=Constant3Vector() {constant=<rgba 0.8963 0.5457 0.0296 1>}
    C=Constant3Vector() {constant=<rgba 0.0931 0.9911 0.0452 1>}
    Alpha=@n2
   ) {materialFunction=&3ColorBlend}.Result
   Alpha=@n1
  )
  n3 = MaterialFunctionCall(Float2=Multiply(A=TextureCoordinate(), B=Constant2Vector() {g=0.1, r=1})) {materialFunction=&BreakOutFloat2Components}
  n4 = NamedRerouteDeclaration(None=@n3.R) {name=UV R, nodeColor=<rgba 1 0 0.0703 1>}
  n5 = OneMinus(None=@n4)
  NamedRerouteDeclaration(None=Ceil(None=Multiply(A=@n5, B=@n4))) {name=NumberMask, nodeColor=<rgba 0 0.5547 1 1>}
  n6 = TextureCoordinate()
  n7 = Constant2Vector() {g=8}
  n8 = MaterialFunctionCall!(UV=Multiply(A=Add(A=@n6, B=@n7), B=Constant2Vector() {g=0.1, r=1}), Mirror Y, Swizzle UV) {materialFunction=/FortUILibrary/MaterialFunctions/Basics/MF_UI_Symmetry_2x1_New}
  n9 = Constant() {r=0.3}
  NamedRerouteDeclaration(None=MaterialFunctionCall(
    DistanceField=TextureSample(UVs=@n8.Symmetry) {samplerType=SAMPLERTYPE_Grayscale, texture=/GreaseClipDunmurryUI/Art/Textures/T_UI_GreaseClipDunmury_Numbers}.RGB
    Threshold=@n9
   ) {materialFunction=&MF_UI_AntiAliasing}.Outer) {name=BackgroundNumber, nodeColor=<rgba 0 0 0 1>}
  n10 = Clamp(None=ScalarParameter() {desc=Dynamic number of digits, group=Variables, parameterName=Number, sliderMax=300, sortPriority=2}) {maxDefault=99}
  n11 = Constant() {r=1}
  n12 = Floor(None=Add(A=@n10) {desc=Show 0})
  n13 = Add(
   A=If(A=NamedRerouteDeclaration(None=@n10) {name=Numbers, nodeColor=<rgba 1 0.7969 0 1>}, "A > B"=@n11, "A == B"=@n11, "A < B"=Constant() {r=2}) {constB=10}
   B=Floor(None=Logarithm10!(X=@n12))
  )
  n14 = NamedRerouteDeclaration(None=Multiply(A=NamedRerouteDeclaration(None=@n13) {name=NumOfDigits, nodeColor=<rgba 0 1 0.4844 1>}, B=@n5) {constA=3, constB=3}) {name=UV Integer, nodeColor=<rgba 0 0.3906 1 1>}
  n15 = Ceil(None=@n14)
  n16 = Divide(A=Abs(None=NamedRerouteUsage()), B=Power(Base=Constant() {r=10}, Exp=Add(A=@n15)) {desc="10-100,000,000"})
  n17 = Floor(None=Multiply(A=Frac(None=@n16) {desc=?}) {constB=100})
  n18 = AppendVector(A=OneMinus(None=Frac(None=@n14)), B=Add(A=NamedRerouteUsage(), B=Divide(A=@n17) {constB=10}))
  NamedRerouteDeclaration(None=MaterialFunctionCall(
    DistanceField=TextureSample(UVs=@n18) {samplerType=SAMPLERTYPE_Grayscale, texture=&T_UI_GreaseClipDunmury_Numbers}.RGB
    Threshold=@n9
   ) {materialFunction=&MF_UI_AntiAliasing}.Outer) {name=SpeedNumber, nodeColor=<rgba 0 0 0 1>}
  NamedRerouteDeclaration(None=@n3.G) {name=UV G, nodeColor=<rgba 0 1 0.25 1>}
 ]
}
